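# Additive unlit material for animated HVAC wind FX, combining three looks selected by vertex color: streaked wind, turbulent wind and a scrolling fan color in red.
# Panned streak and turbulence textures are shaped by curve-atlas ramps and view-angle fading, with timing offset by object position, vertex alpha and a speed parameter.
# The green-channel layer adds noise-driven, distance-adjusted vertex displacement.
# Quality and shading-path switches vary noise, displacement and time-of-day tinting.
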Material /Game/Athena/Environments/Blueprints/DudeBro/HVAC_FX/FX_Wind
graph {
 properties {bUseMaterialAttributes=true, blendMode=BLEND_Additive, pixelDepthOffsetMode=PDOM_Legacy, shadingModel=MSM_Unlit, usedShadingModels=Unlit}
 nodes [
  n1 = TextureCoordinate() {uTiling=6, vTiling=0.5}
  n2 = VectorParameter() {parameterName=VU3, defaultValue=<rgba 0.3333 1 1 1>}
  n3 = ComponentMask(None=@n2.RGB) {g=true, r=true}
  n4 = Time()
  n5 = ScalarParameter() {parameterName=BP_SpeedUpTimer}
  n6 = ObjectPositionWS()
  n7 = Add(
   A=Add(
    A=Add(A=@n4, B=@n5)
    B=MaterialFunctionCall!(f2, f3=@n6, f4) {materialFunction=/Game/Packages/Fortnite_UniqueMaterials_1/VectorComponentOps}.f3 Add
   )
   B=Multiply(A=VertexColor().A) {constB=317.4286}
  )
  n8 = TextureSampleParameter2D(UVs=Panner(Coordinate=Multiply(A=@n1, B=@n3), Time=@n7, Speed=AppendVector(A=@n2.B, B=@n2.B)) {speedX=0.1, speedY=0.5}) {
   parameterName=Param
   samplerSource=SSM_Wrap_WorldGroupSettings
   texture=/Game/Effects/Fort_Effects/Textures/Energy/T_OutpostShockwaveStreaks_02
  }
  n9 = MaterialFunctionCall(Float2=TextureCoordinate() {coordinateIndex=1}) {materialFunction=&BreakOutFloat2Components}
  n10 = CurveAtlasRowParameter(CurveTime=@n9.G) {
   atlas=&AthenaCurveAtlas01
   curve=&/Curve_Ramps2
   defaultValue=88
  }
  n11 = MaterialFunctionCall(Float2=TextureCoordinate() {uTiling=6, vTiling=0.5}) {materialFunction=&BreakOutFloat2Components}
  n12 = AppendVector(A=@n11.R, B=Multiply(A=@n11.G, B=@n11.G))
  n13 = TextureSample() {
   samplerSource=SSM_Wrap_WorldGroupSettings
   samplerType=SAMPLERTYPE_LinearColor
   texture=&T_NoisePack_LinearColor
  }
  n14 = Subtract(A=AppendVector(A=@n13.R, B=@n13.R)) {constB=0.35}
  n15 = ShadingPathSwitch(Default=Add(A=@n12, B=Multiply(A=@n14) {constB=0.05}), Mobile=@n12) {desc=Migrated from FeatureLevelSwitch}
  n16 = VectorParameter() {parameterName=VU, defaultValue=<rgba 0.3 0.3 0 0.3>}
  n17 = Multiply(A=@n15, B=ComponentMask(None=@n16.RGB) {g=true, r=true})
  n18 = AppendVector(A=@n16.B, B=@n16.B)
  n19 = VectorParameter() {parameterName=VU2, defaultValue=<rgba 0.25 0.8 0.1 0.5>}
  n20 = ComponentMask(None=@n19.RGB) {g=true, r=true}
  n21 = TextureSample(UVs=Panner(Coordinate=Multiply(A=@n15, B=@n20), Time=@n7, Speed=AppendVector(A=@n19.B, B=@n19.B)) {speedX=0.1, speedY=0.5}) {samplerSource=SSM_Wrap_WorldGroupSettings, texture=/Game/Global/Textures/Default/Caustics/T_Turbulence_Seamless}
  n22 = Multiply(
   A=TextureSample(UVs=Panner(Coordinate=@n17, Time=@n7, Speed=@n18) {speedY=0.3}) {samplerSource=SSM_Wrap_WorldGroupSettings, texture=&T_Turbulence_Seamless}.R
   B=@n21.R
  )
  n23 = Multiply(A=@n21.G, B=@n21.G)
  n24 = QualitySwitch(Default=@n22, Low=@n23, High=@n23, Medium=@n23, Epic=@n22)
  n25 = CurveAtlasRowParameter(CurveTime=@n9.G) {
   atlas=&AthenaCurveAtlas01
   curve=&/Curve_Ramps
   defaultValue=84
  }
  n26 = Multiply(A=Max(A=Multiply(A=@n8.R, B=@n10.R), B=@n24), B=@n25.B)
  n27 = Normalize(VectorInput=Subtract(A=WorldPosition().XYZ, B=CameraPositionWS()))
  n28 = VertexNormalWS()
  n29 = CurveAtlasRowParameter(CurveTime=Abs(None=DotProduct(A=@n27, B=@n28))) {
   atlas=&AthenaCurveAtlas01
   curve=&Curve_Ramps
   defaultValue=84
  }
  n30 = SetMaterialAttributes!(MaterialAttributes, Base Color, Opacity=Multiply(A=Multiply(A=@n26) {constB=0.3}, B=@n29.G)) {attributeSetTypes [#1, #2]}
  n31 = CurveAtlasRowParameter(CurveTime=@n9.R) {
   atlas=&AthenaCurveAtlas01
   curve=&Curve_Ramps
   defaultValue=84
  }
  n32 = Multiply(A=@n31.A, B=@n25.B)
  n33 = TextureCoordinate()
  n34 = Add(A=@n33, B=Divide(A=@n7, B=Constant2Vector() {g=512, r=16}))
  n35 = TextureSample(UVs=@n34) {
   samplerSource=SSM_Wrap_WorldGroupSettings
   samplerType=SAMPLERTYPE_LinearColor
   texture=/Game/Effects/Fort_Effects/Textures/Tile/LinerNoise64x64
  }
  n36 = TextureSample!(UVs=@n34, Tex, Level=ScalarParameter() {defaultValue=2, parameterName=MipLevel}, Apply View MipBias) {
   mipValueMode=TMVM_MipLevel
   samplerSource=SSM_Wrap_WorldGroupSettings
   samplerType=SAMPLERTYPE_LinearColor
   texture=&LinerNoise64x64
  }
  n37 = Multiply(
   A=OneMinus(None=ComponentMask(None=@n33) {g=true})
   B=Add(
    A=ShadingPathSwitch(Default=@n35.RGB, Mobile=@n36.RGB) {desc=Migrated from FeatureLevelSwitch}
    B=Constant3Vector() {constant=<rgba -0.3 -0.3 0.5 0>}
   )
  )
  n38 = VectorParameter() {parameterName=WPO_Size, defaultValue=<rgba 75 75 250 1>}
  n39 = ScalarParameter() {defaultValue=150, parameterName="Fade_Start_Distance (S)"}
  n40 = ScalarParameter() {defaultValue=50, parameterName="Fade_End_Distance (S)"}
  n41 = LinearInterpolate(
   A=Constant() {r=-15}
   Alpha=Clamp(None=MaterialFunctionCall(Fade_Start_Distance=@n39, Fade_End_Distance=@n40) {materialFunction=&MF_Distance_SyncedFade_FX}.Result)
  )
  n42 = Clamp(None=@n41) {minDefault=-2}
  n43 = QualitySwitch(Default=Constant() {r=1}, Low=@n42, High=@n42, Medium=@n42, Epic=@n42)
  n44 = ScalarParameter() {defaultValue=1, parameterName=BP_WPO_MovingUP}
  n45 = Clamp(None=Multiply(A=@n44, B=@n41)) {maxDefault=10, minDefault=-2}
  n46 = MaterialFunctionCall(
   X=Multiply(A=@n38.R, B=@n43)
   Y=Multiply(A=@n38.G, B=@n43)
   Z=Multiply(A=@n38.B, B=QualitySwitch(Default=@n44, Low=@n45, High=@n45, Medium=@n45, Epic=@n45))
  ) {materialFunction=&MakeFloat3}
  n47 = SetMaterialAttributes!(
   MaterialAttributes
   Base Color
   Opacity=Multiply(A=@n29.G, B=Multiply(A=@n24, B=@n32))
   World Position Offset=Transform(None=Multiply(A=@n37, B=@n46.Result)) {transformSourceType=TRANSFORMSOURCE_Local}
  ) {attributeSetTypes [#1, #2, #3]}
  n48 = VertexColor()
  n49 = VectorParameter() {parameterName=Color, defaultValue=<rgba 0.2804 0.4184 0.5 1>}
  n50 = Multiply(A=@n49.RGB, B=ScalarParameter() {defaultValue=1, parameterName=BP_Intensity})
  n51 = Multiply(A=Multiply(A=ScalarParameter() {defaultValue=0.5, parameterName=EmmisiveUnderEpic}, B=@n49.RGB), B=@n50)
  n52 = Multiply(A=@n50, B=MaterialFunctionCall() {materialFunction=&Color_Time_of_Day}.Result) {constB=0.5}
  n53 = MaterialFunctionCall(Float2=TextureCoordinate()) {materialFunction=&BreakOutFloat2Components}
  n54 = CurveAtlasRowParameter(CurveTime=@n53.G) {
   atlas=&AthenaCurveAtlas01
   curve=&/Curve_FanColor
   defaultValue=83
  }
  n55 = VectorParameter() {parameterName=FanColor, defaultValue=<rgba 1 0.8784 0 1>}
  n56 = CurveAtlasRowParameter(CurveTime=Frac(None=Add(A=@n53.R, B=@n7))) {
   atlas=&AthenaCurveAtlas01
   curve=&Curve_Ramps
   defaultValue=84
  }
  n57 = Multiply(A=@n54.A, B=@n56.R)
  BlendMaterialAttributes(
   A=SetMaterialAttributes!(
    MaterialAttributes=BlendMaterialAttributes(A=@n30, B=@n47, Alpha=@n48.G)
    Emissive Color=QualitySwitch(Default=@n51, Low=@n51, High=@n51, Medium=@n51, Epic=@n52)
    Base Color=@n49.RGB
   ) {attributeSetTypes [#4, #1]}
   B=SetMaterialAttributes!(MaterialAttributes, Base Color=Multiply(A=@n55.RGB, B=@n54), Emissive Color, Opacity=Multiply(A=@n55.A, B=@n57)) {attributeSetTypes [#1, #4, #2]}
   Alpha=@n48.R
  )
  Multiply(A=@n10.R)
 ]
}
