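# M_WindStrip is a translucent, two-sided particle material for wind streaks, built on a soft smoke trail texture.
# UVs can be swizzled and tiled, then scroll by a particle dynamic Time parameter, and a side mask shapes the edges.
# Particle color tints the result, with secondary-color gradient, time-of-day emissive and depth fade nodes.
# A static switch enables SplineThicken world position offset, with base width driven by the dynamic WPO parameter.
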Material /Game/Athena/Environments/Blueprints/DudeBro/HVAC_FX/M_WindStrip
graph {
 properties {
  bUsedWithMeshParticles=true
  bUsedWithNanite=true
  bUsedWithNiagaraMeshParticles=true
  bUsedWithParticleSprites=true
  blendMode=BLEND_Translucent
  pixelDepthOffsetMode=PDOM_Legacy
  twoSided=true
  usedShadingModels=Default Lit
 }
 nodes [
  n1 = TextureCoordinate()
  n2 = StaticSwitchParameter(True=MaterialFunctionCall(XY=@n1) {materialFunction=&Swizzle}.YX, False=@n1) {parameterName=SwizeUV}
  n3 = Multiply(A=@n2, B=AppendVector(A=ScalarParameter() {defaultValue=0.5, parameterName=TileX}, B=Constant() {r=1}))
  n4 = Multiply(A=@n3) {constB=2}
  n5 = ComponentMask(None=@n4) {r=true}
  n6 = Divide(B=Sine(None=ComponentMask(None=@n3) {r=true}))
  n7 = Multiply(A=Subtract(A=ComponentMask(None=@n4) {g=true})) {constB=0.5}
  Add(A=AppendVector(A=Multiply(A=@n5), B=Multiply(A=@n6, B=@n7))) {constB=0.5}
  n8 = DynamicParameter() {paramNames [Time,Random,Param3,WPO]}
  n9 = Add(A=@n3, B=AppendVector(A=Add(A=@n8.Time, B=@n8.Time), B=Constant()))
  n10 = TextureCoordinate()
  n11 = AppendVector(A=ScalarParameter() {defaultValue=1, parameterName=NoiseTileX}, B=ScalarParameter() {defaultValue=1, parameterName=NoiseTileY})
  n12 = ComponentMask(None=TextureSample(UVs=Multiply(A=@n10, B=@n11)) {samplerType=SAMPLERTYPE_LinearColor, texture=&T_NoisePack_LinearColor}.RGB) {g=true, r=true}
  Add(A=@n9, B=Multiply(A=Subtract(A=@n12) {constB=0.5}, B=ScalarParameter() {defaultValue=0.1, parameterName=DistortionMult}) {constB=0.1})
  n13 = TextureSampleParameter2D(UVs=@n9) {parameterName=MainTexture, texture=/Game/Effects/Fort_Effects/Textures/Smoke/T_Smoke_Trail_Soft_Chromatic}
  n14 = TextureSample(UVs=@n2) {texture=/Game/Effects/Fort_Effects/Textures/Masks/T_Side_Mask}
  n15 = ScalarParameter() {defaultValue=1.5, parameterName=Gradiant Power}
  n16 = MaterialFunctionCall(Float3=Multiply(
    A=StaticComponentMaskParameter(None=@n13.RGB) {defaultB=true, defaultG=true, defaultR=true, parameterName=ChannelMask}
    B=Power(Base=@n14.R, Exp=@n15)
   )) {materialFunction=&BreakOutFloat3Components}
  n17 = ParticleColor()
  DepthFade(Opacity=Multiply(A=@n16.G, B=@n17.A), FadeDistance=Multiply(A=Constant() {r=125}, B=@n8.Param3))
  n18 = Multiply(A=@n16.R, B=@n17.RGB)
  n19 = ComponentMask(None=@n2) {r=true}
  n20 = ScalarParameter() {defaultValue=2, parameterName=GradientPower}
  LinearInterpolate(
   A=@n18
   B=VectorParameter() {parameterName=SecondaryColor}.RGB
   Alpha=Multiply(A=Power(Base=@n19, Exp=@n20), B=ScalarParameter() {defaultValue=1, parameterName=GradientStrength})
  )
  n21 = MaterialFunctionCall() {materialFunction=&Color_Time_of_Day}
  MaterialFunctionCall(Emissive=Multiply(A=LinearInterpolate(A=@n21.Result), B=@n18)) {materialFunction=&MF_FX_EmissiveAdjustments}
  StaticSwitchParameter(
   True=MaterialFunctionCall(WidthBase=@n8.WPO, WidthTip=Constant() {r=1}) {materialFunction=&SplineThicken}.WorldPositionOffset
   False=Constant()
  ) {parameterName=WPO}
 ]
}
